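# Masked, two-sided lit Niagara material for the paper lantern mesh in the floating lantern aura cosmetic.
# Diffuse with hue shift, packed spec masks and normals feed the surface, while masked emissive scales by dynamic intensity and particle color.
# An erosion layer, off by default, combines bounds gradient, noise and world-aligned cracks with dynamic Erode to clip, darken and push vertices along normals.
# A shared emissive adjustment function processes the blended attributes.
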Material /SparksCosmetics/Items/Auras/Assets/NS_Aura_FloatingLantern/M_Aura_Lantern_Mesh
graph {
 properties {
  bUseMaterialAttributes=true
  bUsedWithNanite=true
  bUsedWithNiagaraMeshParticles=true
  bUsedWithNiagaraRibbons=true
  bUsedWithNiagaraSprites=true
  blendMode=BLEND_Masked
  opacityMaskClipValue=0.16
  pixelDepthOffsetMode=PDOM_Legacy
  translucencyLightingMode=TLM_VolumetricPerVertexDirectional
  twoSided=true
  usedShadingModels=Default Lit
 }
 nodes [
  n1 = TextureSampleParameter2D() {
   group=Specular
   parameterName=SpecMasks
   samplerSource=SSM_Wrap_WorldGroupSettings
   samplerType=SAMPLERTYPE_Masks
   texture=/SparksCosmetics/Items/Auras/Assets/NS_Aura_FloatingLantern/Textures/T_PaperLantern_A_S
  }
  n2 = MaterialFunctionCall(Color=@n1.RGB) {materialFunction=&SplitComponents}
  n3 = VectorParameter() {group=Emissive, parameterName="Emissive Color && Intensity", sortPriority=2, defaultValue=<rgba 4.9472 5 3.6115 1>}
  n4 = TextureSampleParameter2D() {
   group=Diffuse
   parameterName=Diffuse
   samplerSource=SSM_Wrap_WorldGroupSettings
   sortPriority=1
   texture=/SparksCosmetics/Items/Auras/Assets/NS_Aura_FloatingLantern/Textures/T_PaperLantern_A_D
  }
  n5 = MaterialFunctionCall(
   Hue Shift Percentage=ScalarParameter() {group=Diffuse, parameterName=Hue Shift Percentage, sortPriority=4}
   Texture=Multiply(A=ScalarParameter() {defaultValue=1, group=Diffuse, parameterName=Diff_Mult, sortPriority=3}, B=@n4.RGB)
  ) {materialFunction=&HueShift}
  n6 = Multiply(A=@n5.Result, B=@n3.RGB)
  n7 = ScalarParameter() {defaultValue=1, group=Emissive, parameterName=UseDiffuseinEmissive, sortPriority=1}
  n8 = TextureSampleParameter2D() {
   group=Emissive
   parameterName=Emissive Mask Texture
   samplerSource=SSM_Wrap_WorldGroupSettings
   samplerType=SAMPLERTYPE_Masks
   sortPriority=3
   texture=/SparksCosmetics/Items/Auras/Assets/NS_Aura_FloatingLantern/Textures/T_PaperLantern_A_M
  }
  n9 = ChannelMaskParameter(None=@n8.RGBA) {group=Emissive, maskChannel=Green, parameterName=Emissive Mask Texture Channel, sortPriority=4, defaultValue=<rgba 0 1 0 0>}
  n10 = LinearInterpolate(
   A=@n9.R
   B=ComponentMask(None=@n8.RGBA) {b=true, g=true, r=true}
   Alpha=ScalarParameter() {group=Emissive, parameterName=Emissive Mask Uses RGB, sortPriority=5}
  )
  n11 = Multiply(
   A=Multiply(
    A=LinearInterpolate(A=@n3.RGB, B=@n6, Alpha=@n7)
    B=DynamicParameter() {defaultValue=<rgba 1 1 0.2832 1>, paramNames [Emissive Intensity, --, Erode, --]}.Emissive Intensity
   )
   B=LinearInterpolate(A=@n10, B=OneMinus(None=@n10), Alpha=ScalarParameter() {group=Emissive, parameterName=Invert Emissive Mask, sortPriority=6})
  )
  n12 = ParticleColor()
  n13 = ScalarParameter() {defaultValue=1, group=Diffuse, parameterName=Diffuse Alpha Blend, sortPriority=3}
  n14 = Multiply(
   A=TextureCoordinate()
   B=ComponentMask(None=VectorParameter() {group=Opacity, parameterName=Opacity Mask UV Scale, sortPriority=3, defaultValue=<rgba 1 1 0 0>}.RGB) {g=true, r=true}
  )
  n15 = TextureSampleParameter2D(UVs=LinearInterpolate(
    A=@n14
    B=MaterialFunctionCall(XY=@n14) {materialFunction=&Swizzle}.YX
    Alpha=ScalarParameter() {group=Opacity, parameterName=Swizzle Opacity Mask UVs, sortPriority=4}
   )) {
   group=Opacity
   parameterName=Opacity_Mask_Texture
   samplerSource=SSM_Wrap_WorldGroupSettings
   sortPriority=1
   texture=&T_WhitesRGB
  }
  n16 = Multiply(
   A=ParticleColor().A
   B=Multiply(
    A=LinearInterpolate(B=@n4.A, Alpha=@n13) {constA=1}
    B=ChannelMaskParameter(None=@n15.RGBA) {group=Opacity, parameterName=Opacity_Mask Channel Mask, sortPriority=2}.R
   )
  )
  n17 = Multiply(
   A=MaterialFunctionCall(Distance=ScalarParameter() {group=Opacity, parameterName=Near Camera Fade Distance}) {materialFunction=&Near_Camera_Fade}.Result
   B=@n16
  )
  n18 = ScalarParameter() {defaultValue=25, group=Opacity, parameterName=DepthFadeDistance}
  n19 = TextureSampleParameter2D() {
   group=Normal
   parameterName=Normals
   samplerSource=SSM_Wrap_WorldGroupSettings
   samplerType=SAMPLERTYPE_Normal
   texture=/SparksCosmetics/Items/Auras/Assets/NS_Aura_FloatingLantern/Textures/T_PaperLantern_A_N
  }
  n20 = ParticleColor()
  n21 = MakeMaterialAttributes!(
   BaseColor=LinearInterpolate(
    A=@n5.Result
    B=Multiply(A=@n5.Result, B=@n20.RGB)
    Alpha=ScalarParameter() {defaultValue=1, group=Diffuse, parameterName=Apply Particle Color To Diffuse}
   )
   Metallic=@n2.G
   Specular=@n2.G
   Roughness=LinearInterpolate(
    A=ScalarParameter() {group=Specular, parameterName=SpecRoughnessMin, sortPriority=1}
    B=ScalarParameter() {defaultValue=1, group=Specular, parameterName=SpecRoughnessMax, sortPriority=2}
    Alpha=@n2.B
   )
   Anisotropy
   EmissiveColor=LinearInterpolate(
    A=@n11
    B=Multiply(A=@n11, B=@n12.RGB)
    Alpha=ScalarParameter() {defaultValue=1, group=Emissive, parameterName=Apply Particle Color To Emisive}
   )
   Opacity=StaticSwitchParameter(True=DepthFade(Opacity=@n17, FadeDistance=@n18), False=@n17) {group=Opacity, parameterName=UseDepthFade}
   OpacityMask=@n16
   Normal=MaterialFunctionCall(Normal=@n19.RGB, Flatness=ScalarParameter() {group=Normal, parameterName=Normal Flatness}) {materialFunction=&FlattenNormal}.Result
   Tangent
   WorldPositionOffset
   SubsurfaceColor
   ClearCoat
   ClearCoatRoughness
   AmbientOcclusion
   Refraction
   CustomizedUVs_0
   CustomizedUVs_1
   CustomizedUVs_2
   CustomizedUVs_3
   CustomizedUVs_4
   CustomizedUVs_5
   CustomizedUVs_6
   CustomizedUVs_7
   PixelDepthOffset
   ShadingModel
   Displacement
  )
  n22 = GetMaterialAttributes(None=@n21) {attributeGetTypes [
    #1
    #2
    #3
    #4
    #5
   ]}
  n23 = VectorParameter() {group=Erosion, parameterName=Crack Color, sortPriority=11, defaultValue=<rgba 0.15 0.15 0.15 0>}
  n24 = LinearInterpolate(
   A=TextureCoordinate()
   B=TextureCoordinate() {coordinateIndex=1}
   Alpha=ScalarParameter() {defaultValue=1, group=Erosion, parameterName=Use UV1 Crack Texture, sortPriority=7}
  )
  n25 = TextureObjectParameter() {group=Erosion, parameterName=Crack, sortPriority=6, texture=/Game/Global/Textures/Default/Masks_and_Gradients/Cracks_02}
  n26 = ScalarParameter() {defaultValue=350, group=Erosion, parameterName=Crack World Aligned Texture Size, sortPriority=9}
  n27 = WorldPosition() {worldPositionShaderOffset=WPT_ExcludeAllShaderOffsets}
  n28 = Multiply(
   A=LinearInterpolate(
    A=TextureSample(UVs=@n24, Tex=@n25) {texture=&Cracks_02}.RGB
    B=MaterialFunctionCall(TextureObject=@n25, TextureSize=@n26, WorldPosition=@n27.XYZ) {materialFunction=&WorldAlignedTexture}.XYZ Texture
    Alpha=ScalarParameter() {defaultValue=1, group=Erosion, parameterName=Use WorldAligned Crack Texture, sortPriority=8}
   )
   B=ScalarParameter() {defaultValue=1, group=Erosion, parameterName=CrackSize, sortPriority=10}
  )
  n29 = ComponentMask(None=@n28) {r=true}
  n30 = Add(A=ObjectRadius()) {constB=0.001}
  n31 = Add(
   A=Subtract(A=WorldPosition() {worldPositionShaderOffset=WPT_ExcludeAllShaderOffsets}.XYZ, B=ObjectPositionWS()) {desc=Doesn't break with different locations}
   B=@n30
  )
  n32 = Multiply(A=@n30) {constB=2}
  n33 = Saturate(None=ComponentMask(None=Divide(A=@n31, B=@n32)) {b=true})
  n34 = LinearInterpolate(
   A=Constant() {r=0.5}
   B=LinearInterpolate(A=OneMinus(None=@n33), B=@n33, Alpha=ScalarParameter() {defaultValue=1, group=Erosion, parameterName=BoundRampDirection, sortPriority=5})
   Alpha=ScalarParameter() {defaultValue=1, group=Erosion, parameterName=Use Bounds in Erosion, sortPriority=14}
  )
  n35 = ComponentMask(None=@n28) {b=true}
  n36 = Multiply(A=@n35, B=ScalarParameter() {defaultValue=0.5, group=Erosion, parameterName=CrackAffectMask, sortPriority=5}) {constB=0.25}
  n37 = Multiply(
   A=TextureCoordinate() {coordinateIndex=1}
   B=ComponentMask(None=VectorParameter() {group=Erosion, parameterName=Noise_UV_Scale, sortPriority=4, defaultValue=<rgba 2 2 0 0>}.RGB) {g=true, r=true}
  )
  n38 = TextureSampleParameter2D(UVs=LinearInterpolate(
    A=@n37
    B=MaterialFunctionCall(XY=@n37) {materialFunction=&Swizzle}.YX
    Alpha=ScalarParameter() {group=Erosion, parameterName=Swizzle Erosion Noise, sortPriority=5}
   )) {
   group=Erosion
   parameterName=Erosion Noise
   samplerType=SAMPLERTYPE_LinearColor
   sortPriority=3
   texture=/Game/Characters/Player/Male/Medium/Bodies/M_MED_SRDrift/Textures/T_ElectricNoise
  }
  n39 = ScalarParameter() {defaultValue=1, group=Erosion, parameterName=NoiseAffectMask, sortPriority=2}
  n40 = Subtract(
   A=Add(A=Subtract(A=@n34, B=@n36), B=Multiply(A=@n38.R, B=@n39))
   B=DynamicParameter() {defaultValue=<rgba 1 1 0.2832 1>, paramNames [Emissive Intensity, --, Erode, --]}.Erode
  )
  n41 = Multiply(A=@n40) {constB=-1}
  n42 = Saturate(None=Add(A=@n41) {constB=0.5})
  n43 = ScalarParameter() {defaultValue=0.1, group=Erosion, parameterName=Crack Specular, sortPriority=12}
  n44 = ScalarParameter() {defaultValue=1, group=Erosion, parameterName=Crack Roughness, sortPriority=13}
  n45 = Multiply(A=Saturate(None=@n41), B=VertexNormalWS())
  n46 = ScalarParameter() {defaultValue=25, group=Erosion, parameterName=Erosion_WPO_Strength, sortPriority=1}
  n47 = SetMaterialAttributes!(
   MaterialAttributes=@n22.MaterialAttributes
   Base Color=LinearInterpolate(A=LinearInterpolate(A=@n22.Base Color, B=@n23.RGB, Alpha=@n29), Alpha=@n42) {constB=0.1}
   Specular=LinearInterpolate(A=LinearInterpolate(A=@n22.Specular, B=@n43, Alpha=@n35), Alpha=@n42) {constB=0.15}
   Roughness=LinearInterpolate(A=LinearInterpolate(A=@n22.Roughness, B=@n44, Alpha=@n35), Alpha=@n42)
   Opacity Mask=Multiply(A=@n22.Opacity Mask, B=@n40)
   World Position Offset=ShadingPathSwitch(Default=Multiply(A=@n45, B=@n46) {constB=100}, Mobile=Constant())
  ) {attributeSetTypes [
    #1
    #2
    #3
    #4
    #5
   ]}
  MaterialFunctionCall!(In=BlendMaterialAttributes(A=@n21, B=@n47, Alpha=ScalarParameter() {group=Erosion, parameterName=Use Advanced Erosion})) {materialFunction=/Game/Effects/Niagara/MasterMaterials/Functions/MF_FX_MAEmissiveAdjustments}
 ]
}
