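# Lit ribbon material for the Floating Lantern aura's Niagara effect, shading a ribbon that blends from particle-tinted non-metal to copper-colored metal.
# A remap of the U coordinate over a short transition range drives both the metallic value and the color blend.
# A wave texture sampled with swapped, tiled UVs adds subtle variation to color brightness and roughness.
# A Camera Offset function shifts the ribbon relative to the camera.
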Material /SparksCosmetics/Items/Auras/Assets/NS_Aura_FloatingLantern/M_Aura_Lantern_Ribbon_Metallic
graph {
 properties {bAutomaticallySetUsageInEditor=false, bUsedWithNiagaraRibbons=true, usedShadingModels=Default Lit}
 nodes [
  n1 = MaterialFunctionCall(
   Input=ComponentMask(None=TextureCoordinate()) {r=true}
   Input Low=ScalarParameter() {parameterName=Transition Start}
   Input High=ScalarParameter() {defaultValue=0.05, parameterName=Metallic End}
  ) {materialFunction=&RemapValue_0-1}
  LinearInterpolate(B=ScalarParameter() {defaultValue=1, parameterName=Metallic}, Alpha=@n1.Result) {constB=0}
  n2 = ScalarParameter() {defaultValue=0.3, parameterName=Roughness}
  n3 = Constant2Vector() {g=3, r=0.5}
  n4 = MaterialFunctionCall(Float2=TextureCoordinate()) {materialFunction=&BreakOutFloat2Components}
  n5 = AppendVector(A=@n4.G, B=OneMinus(None=@n4.R))
  n6 = TextureSample(UVs=Add(A=Multiply(A=@n3, B=@n5), B=Constant2Vector() {r=0.25})) {texture=/SparksCosmetics/Items/Auras/Assets/NS_Aura_FloatingLantern/Textures/T_Aura_Lantern_Wave}
  LinearInterpolate(A=Add(A=@n2) {constB=0.1}, B=Add(A=@n2) {constB=-0.1}, Alpha=@n6.B)
  MaterialFunctionCall(Offset Amount=ScalarParameter() {defaultValue=-10, parameterName=Camera Offset}) {materialFunction=/Engine/Functions/Engine_MaterialFunctions02/WorldPositionOffset/CameraOffset}
  Multiply(A=@n6.B) {constB=0.1}
  n7 = VectorParameter() {parameterName=Non-Metal Color Mult, defaultValue=<rgba 1 1 1 1>}
  n8 = ParticleColor()
  Multiply(
   A=LinearInterpolate(Alpha=@n6.B) {constA=0.9}
   B=LinearInterpolate(
    A=Multiply(A=@n7.RGB, B=@n8.RGB)
    B=VectorParameter() {parameterName=Metal Color, defaultValue=<rgba 0.4508 0.147 0.0482 1>}.RGB
    Alpha=@n1.Result
   )
  )
  ScalarParameter() {parameterName=Specular}
 ]
}
