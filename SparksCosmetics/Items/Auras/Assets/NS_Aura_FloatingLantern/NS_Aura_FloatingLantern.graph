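# Niagara aura cosmetic that hangs three glowing, swaying lanterns from the character's skeleton, with body glows and falling GPU petals.
# Lanterns bob on spring-driven pendulums, picking random colors and pulsing on the music beat, or on a timer in the front end.
# Ribbons and Sparkles follow the lanterns through particle attribute readers, with ribbons forming pendulum chains.
# User bools adjust body-size and front-end behavior; FN_Debug and Minimal emitters are disabled.
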NiagaraSystem /SparksCosmetics/Items/Auras/Assets/NS_Aura_FloatingLantern/NS_Aura_FloatingLantern
dependencies {
 usedDataInterfaces [
  &NiagaraDataInterfaceArrayFloat
  &NiagaraDataInterfaceColorCurve
  &NiagaraDataInterfaceCurve
  &NiagaraDataInterfaceArrayFloat3
  &NiagaraDataInterfaceArrayColor
  &NiagaraDataInterfaceParticleRead
  /Script/Niagara.NiagaraDataInterfaceSimpleCounter
  &NiagaraDataInterfaceSkeletalMesh
 ]
 usedDynamicInputs [&SimulationPosition]
 usedModules [
  &self:SystemSpawnScript.SystemScriptSource.SystemScriptGraph.NiagaraNodeAssignment_0.SetVariables_#1
  &self:SystemSpawnScript.SystemScriptSource.SystemScriptGraph.NiagaraNodeAssignment_1.SetVariables_#2
  &self:SystemSpawnScript.SystemScriptSource.SystemScriptGraph.NiagaraNodeAssignment_3.SetVariables_#3
  &self:SystemSpawnScript.SystemScriptSource.SystemScriptGraph.NiagaraNodeAssignment_4.SetVariables_#4
  &SystemState
 ]
}
emitters {
 BodyGlow {
  emitterUpdateScript [
   &EmitterState(
    Inactive Response=::NewEnumerator1
    Life Cycle Mode=::NewEnumerator1
    Loop Behavior=::NewEnumerator1
    Loop Duration Mode=::NewEnumerator1
    UseLoopDelay=false
   )
   &SpawnBurst_Instantaneous(Spawn Count=6, Use Spawn Probability=false, UseLoopCountLimit=false)
  ]
  particleSpawnScript [
   &PartitionParticles(Number Of Partitions=6)
   &InitializeParticle(
    Lifetime=&RandomRangeFloat(Minimum=3, Maximum=4)
    Sprite Rotation Mode=::NewEnumerator0
    Sprite Size Mode=::NewEnumerator3
    Uniform Sprite Size=&Multiply_Float(A=40, B=&MakeCustomFloatFromBool(Bool=$User.IsLargeBody, True Float=1.5, False Float=1))
    UsePositionOffset=false
    Write Material Random=true
   ) {module=InitializeParticle001}
   /SetVariables(Particles.glowScale:Vector2f=&Vector2DFromFloat(Value=&Multiply_Float(
      A=&SelectFloatFromArray(
       Direct Array Index=$Particles.PartitionParticles.Partition
       Float Selection Array=DataInterface() {propertyValues {FloatData [90,90,50,50,30,30,150]}}
      )
      B=&MakeCustomFloatFromBool(Bool=$User.IsLargeBody, True Float=1.2, False Float=1)
     ))) {isSetParameters=true}
   /SetVariables(Particles.glowCameraOffset=-10) {isSetParameters=true}
  ]
  particleUpdateScript [
   &ParticleState(Kill Particles When Lifetime Has Elapsed=false, Let Infinitely Lived Particles Die When Emitter Deactivates=true)
   &CameraOffset(Camera Offset Amount=10)
   &SpriteRotationRate(Rotation Rate=&RandomRangeFloat(Minimum=10, Maximum=20))
   &SkeletalMeshLocation(
    Bone / Socket Index=&Add_Integer(A=$Particles.PartitionParticles.Partition, B=1)
    Bone Sampling Mode=::NewEnumerator4
    MaskSpawnGroup=false
    Orientation Sampling=::NewEnumerator1
    Sampled Position Offset=&MakeCustomVectorFromBool(
     Bool=&SetBoolByIntComparison(A=$Particles.PartitionParticles.Partition, B=6, Comparison Type=::NewEnumerator4)
     True Vector=<10 0 0>
     False Vector=<0 0 0>
    )
    Skeletal Mesh=$System.Skeletal Mesh
   ) {module=SkeletalMeshLocation003}
   &InterpolateOverTime(
    Initial Color Value:LinearColor=&SelectLinearColorFromArray(Array Sampling Mode=::NewEnumerator0, Direct Array Index=?, Color Selection Array=$System.ColorPattern, Random Seed=0)
    Rate Of Change=2
    Target Color Value:LinearColor=&SelectLinearColorFromArray(
     Array Sampling Mode=::NewEnumerator1
     Direct Array Index=&Add_Integer(A=$Particles.PartitionParticles.Partition, B=&MakeIntFromFloat(Float=$Engine.Time))
     Direct Array Mode=::NewEnumerator1
     Color Selection Array=$System.ColorPattern
     Random Seed=?
    )
    Variable Type=::NewEnumerator4
   )
   &ScaleColor(
    Scale RGB=&MakeVectorFromLinearColorRGB(Color=$Particles.InterpolateOverTime.InterpolateOverTime_LinearColor.Moving Average)
    ScaleA=true
    ScaleRGB=true
   )
   /SetVariables(Particles.decalColor:LinearColor=&Multiply_LinearColorByFloat(
     Linear Color=$Particles.InterpolateOverTime.InterpolateOverTime_LinearColor.Moving Average
     Float=&MakeCustomFloatFromBool(Bool=$User.IsFrontEnd, True Float=0.75, False Float=1)
    )) {isSetParameters=true}
   &DynamicMaterialParameters(
    Index 0 Param 1=&SelectFloatFromArray(Direct Array Index=$Particles.PartitionParticles.Partition, Float Selection Array=DataInterface() {propertyValues {FloatData [5,5,5,3,3]}})
    Index 0 Param 2=0.1
    Index 0 Param 3=1
    Module.Param0WriteEnabled=true
    Module.Param1WriteEnabled=true
    Module.Param2WriteEnabled=true
    Module.Param3WriteEnabled=true
   )
  ]
  rendererClasses [&NiagaraSpriteRendererProperties]
  renderers [
   {rendererClass=&NiagaraSpriteRendererProperties}
   {rendererClass=&NiagaraSpriteRendererProperties, rendererIndex=1}
   {rendererClass=&NiagaraSpriteRendererProperties, rendererIndex=2}
  ]
 }
 FN_Debug {
  bEnabled=false
  emitterUpdateScript [
   &EmitterState()
   &SpawnBurst_Instantaneous(Use Spawn Probability=false, UseLoopCountLimit=true)
  ]
  particleSpawnScript [&InitializeParticle(
    Lifetime=5
    Mass Mode=::NewEnumerator1
    Position Offset=<40 0 160>
    Sprite Size=<75 75>
    Sprite Size Mode=::NewEnumerator4
    UsePositionOffset=true
    Write Material Random=true
   ) {module=InitializeParticle001}]
  particleUpdateScript [
   &ParticleState(Kill Particles When Lifetime Has Elapsed=false)
   &self:CountBeat() {module=CountBeat}
   /SetVariables(Particles.Debug:NiagaraFloat=&MakeFloatFromInt(Int=$Particles.BeatCount)) {isSetParameters=true}
   &DynamicMaterialParameters(
    Index 0 Param 1=$Particles.Debug
    Module.Index0Param0Write=true
    Module.Index0Param1Write=true
    Module.Index0Param2Write=true
    Module.Index0Param3Write=true
   )
  ]
  rendererClasses [&NiagaraSpriteRendererProperties]
  renderers [{rendererClass=&NiagaraSpriteRendererProperties}]
 }
 HangingLanterns {
  emitterSpawnScript [/SetVariables(Emitter.EmitterRandom:NiagaraInt32=&RandomRangeInt(Minimum=1, Maximum=500)) {isSetParameters=true}]
  emitterUpdateScript [
   &EmitterState()
   &SpawnBurst_Instantaneous(Spawn Count=3, Use Spawn Probability=false, UseLoopCountLimit=true)
  ]
  particleSpawnScript [
   &InitializeParticle(
    Lifetime=&MakeCustomFloatFromBool(Bool=$User.IsLargeBody, True Float=4, False Float=3)
    Mesh Scale Mode=::NewEnumerator3
    Mesh Uniform Scale=0.35
    Position Offset=&RandomRangeVector(Minimum=<-50 -50 75>, Maximum=<50 -50 75>)
    Sprite Rotation Mode=::NewEnumerator2
    Sprite Size Mode=::NewEnumerator3
    Uniform Sprite Size=50
    UsePositionOffset=true
    Write Material Random=true
   )
   /SetVariables(Particles.MeshIndex=&RandomRangeInt(Minimum=0, Maximum=2)) {isSetParameters=true}
   /SetVariables(Particles.MeshIndex=&ReturnExecIndex()) {enabled=false, isSetParameters=true}
   &CameraOffset(Camera Offset Amount=$System.OffsetLanterns)
  ]
  particleUpdateScript [
   &ParticleState(Kill Particles When Lifetime Has Elapsed=false, Let Infinitely Lived Particles Die When Emitter Deactivates=true)
   /SetVariables(Transient.OverallScale:NiagaraFloat=&FloatFromCurve(
     FloatCurve=DataInterface() {propertyValues {Curve {keys [
         {arriveTangent=17.9219, interpMode=RCIM_Cubic, leaveTangent=17.9219, tangentMode=RCTM_User, time=0, value=0}
         {arriveTangentWeight=0.3859, interpMode=RCIM_Cubic, leaveTangentWeight=0.0155, tangentMode=RCTM_User, time=0.1, value=1.2}
         <at 0.2 1>
        ]}}}
     CurveIndex=$Engine.System.Age
    )) {isSetParameters=true}
   &ScaleMeshSize(Scale Factor=&VectorFromFloat(Value=$Transient.OverallScale))
   &self:CountBeat() {module=CountBeat}
   /SetVariables(
    Transient.ColorBrightness:NiagaraFloat=&MakeCustomFloatFromBool(
     Bool=$User.IsFrontEnd
     True Float=&FloatFromCurve(
      FloatCurve=DataInterface() {propertyValues {Curve {keys [
          {interpMode=RCIM_Cubic, leaveTangentWeight=0.0667, tangentMode=RCTM_User, time=0, value=1}
          {arriveTangentWeight=0.0786, interpMode=RCIM_Cubic, leaveTangentWeight=0.0786, tangentMode=RCTM_User, time=0.2, value=0.3}
          {arriveTangentWeight=0.0772, interpMode=RCIM_Cubic, leaveTangentWeight=0.1544, tangentMode=RCTM_User, time=0.4, value=0.6}
          {arriveTangentWeight=0.1444, interpMode=RCIM_Cubic, leaveTangentWeight=0.0722, tangentMode=RCTM_User, time=0.8, value=0.15}
          {arriveTangentWeight=0.0667, interpMode=RCIM_Cubic, tangentMode=RCTM_User, time=1, value=1}
         ]}}}
      CurveIndex=&ModuloFloat(A=$Engine.Time, B=1)
     )
     False Float=&FloatFromCurve(
      FloatCurve=DataInterface() {propertyValues {Curve {keys [
          {interpMode=RCIM_Cubic, leaveTangentWeight=0.0667, tangentMode=RCTM_User, time=0, value=1}
          {arriveTangentWeight=0.1444, interpMode=RCIM_Cubic, leaveTangentWeight=0.0722, tangentMode=RCTM_User, time=0.8, value=0.15}
          {arriveTangentWeight=0.0667, interpMode=RCIM_Cubic, tangentMode=RCTM_User, time=1, value=1}
         ]}}}
      CurveIndex=$NPC.NPC_Sparks_Audio_Parameters001.BeatLifetime
     )
    )
    Transient.ColorSeedFE:NiagaraInt32=&Multiply_Int(
     A=&MakeIntFromFloat(Float=&Add_Float(
       A=&Multiply_Float(A=$Engine.Time, B=1)
       B=&ReturnNormalizedExecIndex(Normalized Index Scale=7)
      ))
     B=&Add_Integer(A=&ReturnExecIndex(), B=$Emitter.EmitterRandom)
    )
    Transient.ColorSeedGameplay:NiagaraInt32=&Multiply_Int(
     A=$Particles.BeatCount
     B=&Add_Integer(A=&ReturnExecIndex(), B=$Emitter.EmitterRandom)
    )
   ) {isSetParameters=true}
   &InterpolateOverTime(
    Rate Of Change=4
    Target Vector Value:Vector3f=&Multiply_VectorByFloat(
     Vector=&MakeVectorFromLinearColorRGB(Color=&SelectLinearColorFromArray(
       Array Sampling Mode=::NewEnumerator1
       Direct Array Index=&RandomRangeInt(
        Minimum=0
        Maximum=5
        Randomness Mode=::NewEnumerator1
        Random Seed=&MakeCustomIntFromBool(Bool=$User.IsFrontEnd, True INT=$Transient.ColorSeedFE, False INT=$Transient.ColorSeedGameplay)
        Fixed Random Seed=true
       )
       Direct Array Mode=::NewEnumerator1
       Color Selection Array=$System.ColorLantern
       Random Seed=?
      ))
     Float=1
    )
    Variable Type=::NewEnumerator2
   ) {module=InterpolateOverTime001}
   &ScaleColor(
    Scale RGB=&Multiply_VectorByFloat(Vector=$Particles.InterpolateOverTime001.InterpolateOverTime_V3_.Moving Average, Float=$Transient.ColorBrightness)
    ScaleA=false
    ScaleRGB=true
   )
   &ScaleColor(
    Scale RGB=&Multiply_VectorByFloat(
     Vector=&MakeCustomVectorFromBool(
      Bool=$User.IsFrontEnd
      True Vector=&MakeVectorFromLinearColorRGB(Color=&SelectLinearColorFromArray(
        Array Sampling Mode=::NewEnumerator1
        Direct Array Index=&RandomRangeInt(
         Minimum=0
         Maximum=5
         Randomness Mode=::NewEnumerator1
         Random Seed=&Multiply_Int(
          A=&MakeIntFromFloat(Float=&Add_Float(A=$Engine.Time, B=&ReturnNormalizedExecIndex(Normalized Index Scale=5, Normalized Index Includes One=false)))
          B=&Add_Integer(A=&ReturnExecIndex(), B=$Emitter.EmitterRandom)
         )
         Fixed Random Seed=true
        )
        Color Selection Array=$System.ColorLantern
        Random Seed=?
       ))
      False Vector=&MakeVectorFromLinearColorRGB(Color=&SelectLinearColorFromArray(
        Array Sampling Mode=::NewEnumerator1
        Direct Array Index=&RandomRangeInt(
         Minimum=0
         Maximum=5
         Randomness Mode=::NewEnumerator1
         Random Seed=&Multiply_Int(
          A=&Add_Integer(A=&ReturnExecIndex(), B=$Particles.BeatCount)
          B=&Add_Integer(A=&ReturnExecIndex(), B=$Emitter.EmitterRandom)
         )
         Fixed Random Seed=true
        )
        Color Selection Array=$System.ColorLantern
        Random Seed=?
       ))
     )
     Float=$Transient.ColorBrightness
    )
    ScaleA=false
    ScaleRGB=true
   ) {enabled=false, module=ScaleColor002}
   &InterpolateOverTime(
    Rate Of Change=4
    Target Vector Value:Vector3f=&Multiply_VectorByFloat(
     Vector=&MakeVectorFromLinearColorRGB(Color=&SelectLinearColorFromArray(
       Array Sampling Mode=::NewEnumerator1
       Direct Array Index=&RandomRangeInt(Minimum=0, Maximum=5, Randomness Mode=::NewEnumerator1, Random Seed=$Transient.ColorSeedFE, Fixed Random Seed=true)
       Direct Array Mode=::NewEnumerator1
       Color Selection Array=$System.ColorLantern
       Random Seed=?
      ))
     Float=1
    )
    Variable Type=::NewEnumerator2
   ) {enabled=false, module=InterpolateOverTime003}
   &ScaleColor(
    Scale RGB=&Multiply_VectorByFloat(
     Vector=$Particles.InterpolateOverTime003.InterpolateOverTime_V3_.Moving Average
     Float=&FloatFromCurve(
      FloatCurve=DataInterface() {propertyValues {Curve {keys [
          {interpMode=RCIM_Cubic, leaveTangentWeight=0.0667, tangentMode=RCTM_User, time=0, value=1}
          {arriveTangentWeight=0.0786, interpMode=RCIM_Cubic, leaveTangentWeight=0.0786, tangentMode=RCTM_User, time=0.2, value=0.3}
          {arriveTangentWeight=0.0772, interpMode=RCIM_Cubic, leaveTangentWeight=0.1544, tangentMode=RCTM_User, time=0.4, value=0.6}
          {arriveTangentWeight=0.1444, interpMode=RCIM_Cubic, leaveTangentWeight=0.0722, tangentMode=RCTM_User, time=0.8, value=0.15}
          {arriveTangentWeight=0.0667, interpMode=RCIM_Cubic, tangentMode=RCTM_User, time=1, value=1}
         ]}}}
      CurveIndex=&ModuloFloat(A=$Engine.Time, B=1)
     )
    )
    ScaleA=true
    ScaleRGB=true
   ) {enabled=false, module=ScaleColor001}
   /SetVariables(
    Transient.LargeBodyOffset:Vector3f=&SelectVectorFromArray(
     Direct Array Index=&ReturnExecIndex()
     Vector Selection Array=DataInterface() {propertyValues {FloatData [<10 -3 40>, <40 -10 0>, <10 -3 -40>]}}
    )
    Transient.MedBodyOffset:Vector3f=&SelectVectorFromArray(
     Direct Array Index=&ReturnExecIndex()
     Vector Selection Array=DataInterface() {propertyValues {FloatData [<5 -8 40>, <40 -20 0>, <5 -8 -40>]}}
    )
   ) {isSetParameters=true}
   &SkeletalMeshLocation(
    Bone / Socket Index=7
    Bone Sampling Mode=::NewEnumerator4
    MaskSpawnGroup=false
    Orientation Sampling=::NewEnumerator1
    Position Sampling=::NewEnumerator2
    Sampled Position Offset=&Add_Vector(
     A=<30 -25 0>
     B=&MakeCustomVectorFromBool(Bool=$User.IsLargeBody, True Vector=$Transient.LargeBodyOffset, False Vector=$Transient.MedBodyOffset)
    )
    Skeletal Mesh=$System.Skeletal Mesh
   )
   &InterpolateOverTime(
    Rate Of Change=5
    Target Position Value:NiagaraPosition=&AddVectorToPosition(
     Position=$Particles.SkeletalMeshLocation.SampledPosition
     Vector=&Multiply_VectorByFloat(
      Vector=<0 0 5>
      Float=&Waveform(
       Global Phase=&Add_Float(A=$Emitter.Age, B=&ReturnNormalizedExecIndex(Normalized Index Scale=9, Normalized Index Includes One=false))
       "Waveform [1]"=::NewEnumerator0
       "[1] Frequency"=&MakeCustomFloatFromBool(Bool=$User.IsFrontEnd, True Float=0.3, False Float=0.5)
       "[1] Amplitude Min/Max"=&MakeCustomVector2DFromBool(Bool=$User.IsFrontEnd, True Vector2D=<0 1>, False Vector2D=<0 2>)
      )
     )
    )
    Variable Type=::NewEnumerator6
   )
   &PendulumSetup(
    Module.Enable Spring Driver=true
    Pendulum Length=30
    Pendulum Pivot:NiagaraPosition=$Particles.InterpolateOverTime.InterpolateOverTime_Position.Moving Average
    Rigid Pendulum=false
    "Spring Driven Constraint (Tightness)":NiagaraFloat=&MakeCustomFloatFromBool(Bool=$User.IsFrontEnd, True Float=1, False Float=0.2)
   )
   &GravityForce()
   &SolveForcesAndVelocity(Clamp Velocity=false, Limit Acceleration=false)
   &PendulumConstraint()
   &CalculateAccurateVelocity()
   /SetVariables(Particles.MidPosition:NiagaraPosition=&LerpPosition(A=$Particles.InterpolateOverTime.InterpolateOverTime_Position.Moving Average, B=$Particles.Position, Alpha=0.6)) {isSetParameters=true}
   &UpdateMeshOrientation(
    Facing Direction=&Subtract_Position(A=$Particles.InterpolateOverTime.InterpolateOverTime_Position.Moving Average, B=$Particles.Position)
    Facing Interpolation=false
    Orientation Method=::NewEnumerator1
    Override Quat=false
    Stabilization Mode:ENiagara_OrientVectorCount=::NewEnumerator0
    TransformFacing=false
   )
   /SetVariables(
    Particles.LightColor=&Multiply_LinearColorByFloat(Linear Color=$Particles.Color, Float=10)
    Particles.LightRadius=&Multiply_Float(A=200, B=&MakeFloatFromVector(Vector=$Particles.Scale, Channel=::NewEnumerator0))
   ) {isSetParameters=true}
   /SetVariables(
    Particles.RayColor:LinearColor=&MultiplyLinearColor(
     Linear Color=$Particles.Color
     Scale Factor=&MakeFloat4FromVectorAndFloat(
      "Vector (RGB)"=&VectorFromFloat(Value=&ClampFloat(Float=&Multiply_Float(A=$Particles.Age, B=1), Max=2))
      "Float (Alpha)"=&MakeCustomFloatFromBool(Bool=$User.IsFrontEnd, True Float=0.75, False Float=1)
     )
    )
    Particles.RaySize:Vector2f=&Vector2DFromFloat(Value=&Multiply_Float(A=100, B=$Transient.OverallScale))
   ) {isSetParameters=true}
   &SpriteRotationRate(Rotation Rate=&Multiply_Float(A=&VectorLength(Vector=$Particles.Velocity), B=0.5)) {enabled=false}
   &AlignSpriteToMeshOrientation(Mesh Orientation Relative Sprite Alignment Vector=<1 0 0>, SetAlignment=true, SetFacing=false)
   &Color(Color=&SelectLinearColorFromArray(
     Array Sampling Mode=::NewEnumerator1
     Direct Array Index=$Particles.PartitionParticles.Partition
     Color Selection Array=DataInterface() {propertyValues {ColorData [<rgba 1 0 0 1>, <rgba 0 1 0 1>, <rgba 0 0 1 1>, <rgba 1 1 1 1>]}}
     Random Seed=?
    )) {enabled=false}
  ]
  rendererClasses [
   &NiagaraMeshRendererProperties
   &NiagaraLightRendererProperties
   &NiagaraSpriteRendererProperties
  ]
  renderers [
   {rendererClass=&NiagaraMeshRendererProperties, rendererIndex=0}
   {rendererClass=&NiagaraLightRendererProperties, rendererIndex=1}
   {rendererClass=&NiagaraSpriteRendererProperties, rendererIndex=2}
   {rendererClass=&NiagaraSpriteRendererProperties, rendererIndex=3}
   {rendererClass=&NiagaraMeshRendererProperties, rendererIndex=4}
  ]
 }
 Minimal {
  bEnabled=false
  emitterUpdateScript [
   &EmitterState()
   &SpawnRate(SpawnRate=5, Use Spawn Probability=false)
  ]
  particleSpawnScript [
   &InitializeParticle(Color Mode=::NewEnumerator0, Lifetime=5, UsePositionOffset=false, Write Material Random=true)
   &AddVelocity()
  ]
  particleUpdateScript [
   &SolveForcesAndVelocity(Clamp Velocity=false, Limit Acceleration=false)
   &ParticleState()
   &Color(Color=&ColorFromCurve(
     ColorCurve=DataInterface() {propertyValues {
       AlphaCurve {keys [<at 0 1 RCIM_Cubic>]}
       BlueCurve {keys [<at 0 0 RCIM_Cubic>,<at 0.5 0>,<at 1 1 RCIM_Cubic>]}
       GreenCurve {keys [<at 0 0 RCIM_Cubic>,<at 0.5 1>,<at 1 0 RCIM_Cubic>]}
       RedCurve {keys [<at 0 1 RCIM_Cubic>,<at 0.5 0>,<at 1 0 RCIM_Cubic>]}
      }}
     CurveIndex=$NPC.NPC_Sparks_Audio_Parameters001.BeatLifetime
    ))
   &Color(Color=&ColorFromCurve(
     ColorCurve=DataInterface() {propertyValues {
       AlphaCurve {keys [<at 0 1 RCIM_Cubic>]}
       BlueCurve {keys [<at 0 0 RCIM_Cubic>,<at 0.5 1>,<at 1 1 RCIM_Cubic>]}
       GreenCurve {keys [<at 0 1 RCIM_Cubic>,<at 0.5 1>,<at 1 0 RCIM_Cubic>]}
       RedCurve {keys [<at 0 1 RCIM_Cubic>,<at 0.5 0>,<at 1 1 RCIM_Cubic>]}
      }}
     CurveIndex=$NPC.NPC_Sparks_Audio_Parameters001.MeasureLifetime
    )) {module=Color001}
   &Color(Color=&ColorFromCurve(
     ColorCurve=DataInterface() {propertyValues {
       AlphaCurve {keys [<at 0 1 RCIM_Cubic>]}
       BlueCurve {keys [<at 0 0 RCIM_Cubic>,<at 0.5 1>,<at 1 0.6146 RCIM_Cubic>]}
       GreenCurve {keys [<at 0 0.5 RCIM_Cubic>,<at 0.5 0.4248>,<at 1 1 RCIM_Cubic>]}
       RedCurve {keys [<at 0 1 RCIM_Cubic>,<at 0.5 0.9484>,<at 1 0.4467 RCIM_Cubic>]}
      }}
     CurveIndex=$NPC.NPC_Sparks_Audio_Parameters001.Amplitude
    )) {module=Color002}
  ]
  rendererClasses [&NiagaraSpriteRendererProperties]
  renderers [{rendererClass=&NiagaraSpriteRendererProperties}]
 }
 Petals {
  simTarget=GPUComputeSim
  emitterUpdateScript [
   &EmitterState()
   &SpawnRate(
    Spawn Probability=0.8
    SpawnRate=&RandomRangeFloat(Minimum=5, Maximum=15, Evaluation Type=::NewEnumerator1)
    Use Spawn Probability=true
   )
  ]
  particleSpawnScript [
   &InitializeParticle(
    Lifetime Max=3
    Lifetime Min=2
    Lifetime Mode=::NewEnumerator1
    Mass Max=1.5
    Mass Mode=::NewEnumerator3
    Sprite Rotation Mode=::NewEnumerator2
    Sprite Size Mode=::NewEnumerator1
    Uniform Sprite Size Max=4
    Uniform Sprite Size Min=3
    UsePositionOffset=false
    Write Material Random=true
   )
   &ShapeLocation(
    MaskSpawnGroup=false
    Non Uniform Scale=<2 2 1>
    Offset=<0 -40 175>
    Offset Mode=::NewEnumerator0
    Override Seed=false
    Sphere Radius=30
    UseOwnerScale=true
   )
   &InitialMeshOrientation(Use Rotation Vector=false)
  ]
  particleUpdateScript [
   &ParticleState()
   &GravityForce(Gravity=<0 0 -750>)
   &AerodynamicDrag(
    Aerodynamic Drag:NiagaraFloat=&RandomRangeFloat(Minimum=0.3333, Maximum=1)
    Aerodynamic Rotational Drag=0.15
    Lift Contribution=0.75
    Pivot Offset Maximum:Vector3f=<5 0.5 1>
    Pivot Offset Minimum:Vector3f=<-5 -0.5 -1>
    Rotation Strength:NiagaraFloat=&RandomRangeFloat(Minimum=1.8, Maximum=3.2)
    Use Linear Drag=true
    Use Rotational Drag=true
   )
   &SolveForcesAndVelocity(Clamp Velocity=false, Limit Acceleration=false)
   &ScaleSpriteSize(Uniform Curve Sprite Scale=DataInterface() {propertyValues {Curve {keys [
        {arriveTangent=14.5835, interpMode=RCIM_Cubic, leaveTangent=14.5835, tangentMode=RCTM_User, time=0, value=0}
        {arriveTangent=1.1111, interpMode=RCIM_Cubic, leaveTangent=1.1111, time=0.1, value=1}
        {arriveTangent=-1.1111, interpMode=RCIM_Cubic, leaveTangent=-1.1111, time=0.9, value=1}
        {arriveTangent=-8.7113, interpMode=RCIM_Cubic, leaveTangent=-8.7113, tangentMode=RCTM_User, time=1, value=0}
       ]}}})
   &AlignSpriteToMeshOrientation(SetAlignment=true, SetFacing=true)
  ]
  rendererClasses [&NiagaraSpriteRendererProperties]
  renderers [{rendererClass=&NiagaraSpriteRendererProperties}]
 }
 Ribbons {
  emitterUpdateScript [
   &EmitterState()
   &SpawnBurst_Instantaneous(Spawn Count=12, Spawn Time=0.2, Use Spawn Probability=false, UseLoopCountLimit=true)
   /SetVariables(
    Emitter.Local Attribute Reader=DataInterface() {propertyValues {EmitterBinding {bindingMode=Self, emitterName=None}}}
    Emitter.Parent Attribute Reader=DataInterface() {propertyValues {EmitterBinding {bindingMode=Other, emitterName=HangingLanterns}}}
    Emitter.RaysPerLantern=6
   ) {isSetParameters=true}
  ]
  particleSpawnScript [
   &PartitionParticles(
    Number Of Partitions=3
    Particles Per Partition=4
    Partition Mode:ENiagara_PartitionMode=::NewEnumerator1
    Write Partition to Ribbon ID=::NewEnumerator0
   )
   /SetVariables(Particles.Counter:NiagaraInt32=/Niagara/DynamicInputs/DataInterface/GenerateIntFromCounter(SimpleCounter:NiagaraDataInterfaceSimpleCounter=DataInterface() {propertyValues {GpuSyncMode=None, InitialValue=-1}})) {isSetParameters=true}
   &InitializeParticle(
    Color Mode=::NewEnumerator0
    Lifetime=5
    Ribbon Width Mode=::NewEnumerator1
    Sprite Size Mode=::NewEnumerator3
    Uniform Sprite Size=50
    UsePositionOffset=false
    Write Material Random=true
   )
  ]
  particleUpdateScript [
   &self:GetParentInfo(
    Normalize Color=true
    Parent Attribute Reader:NiagaraDataInterfaceParticleRead=$Emitter.Parent Attribute Reader
    ParentID:NiagaraInt32=$Particles.PartitionParticles.Partition
    Value Adjustment=2
   ) {module=GetParentInfo}
   &ParticleState(Kill Particles When Lifetime Has Elapsed=false)
   &PendulumSetup(
    Module.Enable Spring Driver=true
    Pendulum Length:NiagaraFloat=&SelectFloatFromArray(
     Direct Array Index=$Particles.PartitionParticles.Index In Partition
     Float Selection Array=DataInterface() {propertyValues {FloatData [0,12]}}
    )
    Pendulum Pivot:Vector3f=&self:PickParentPosition(
     Inherited Position:NiagaraPosition=$Particles.ParentPosition
     Offset:Vector3f=<0 0 3>
     Local Attribute Reader:NiagaraDataInterfaceParticleRead=$Emitter.Local Attribute Reader
    )
    Rigid Pendulum=false
    "Spring Driven Constraint (Tightness)":NiagaraFloat=&Lerp_Float(A=0, B=2, Alpha=&Multiply_FloatByInt(Float=0.25, Integer=$Particles.PartitionParticles.Index In Partition))
   )
   &GravityForce()
   &SolveForcesAndVelocity(Clamp Velocity=false, Limit Acceleration=false)
   &PendulumConstraint()
   &CalculateAccurateVelocity()
   &ScaleRibbonWidth(Ribbon Width Scale=&FloatFromCurve(
     FloatCurve=DataInterface() {propertyValues {Curve {keys [<at 0 0.5 RCIM_Constant>,<at 0.2 1>]}}}
     CurveIndex=&Multiply_FloatByInt(Float=0.25, Integer=$Particles.PartitionParticles.Index In Partition)
    ))
   &DynamicMaterialParameters(
    Index 0 Param 3=&MakeFloatFromInt(Int=&ReturnParticleIDAttributes())
    Index 0 Param 4=&MakeFloatFromInt(Int=$Particles.Counter)
    Module.Index0Param0Write=true
    Module.Index0Param1Write=true
    Module.Index0Param2Write=true
    Module.Index0Param3Write=true
   ) {enabled=false}
   &Color(Color=&ColorFromCurve(
     ColorCurve=DataInterface() {propertyValues {
       AlphaCurve {keys [<at 0 1 RCIM_Cubic>]}
       BlueCurve {keys [<at 0 0 RCIM_Cubic>,<at 0.3333 0>,<at 0.6667 0>,<at 1 1 RCIM_Cubic>]}
       GreenCurve {keys [<at 0 0 RCIM_Cubic>,<at 0.3333 1>,<at 0.6667 1>,<at 1 0 RCIM_Cubic>]}
       RedCurve {keys [<at 0 1 RCIM_Cubic>,<at 0.3333 1>,<at 0.6667 0>,<at 1 0 RCIM_Cubic>]}
      }}
     CurveIndex=&Multiply_FloatByInt(Float=0.3333, Integer=$Particles.PartitionParticles.Index In Partition)
    )) {enabled=false, module=Color001}
   &Color(Color=<rgba 0 0 0 1>) {enabled=false}
   /SetVariables(Particles.PositionMod:NiagaraPosition=&LerpPosition(
     A=$Particles.ParentPosition
     B=$Particles.Position
     Alpha=&ClampFloat(Float=&Add_Float(A=$Emitter.Age, B=-0.5))
    )) {isSetParameters=true}
  ]
  rendererClasses [&NiagaraRibbonRendererProperties, &NiagaraSpriteRendererProperties]
  renderers [
   {rendererClass=&NiagaraRibbonRendererProperties}
   {rendererClass=&NiagaraSpriteRendererProperties, rendererIndex=1}
   {rendererClass=&NiagaraSpriteRendererProperties, rendererIndex=2}
  ]
 }
 Sparkles {
  emitterUpdateScript [
   &EmitterState()
   &SpawnRate(SpawnRate=100, Use Spawn Probability=false)
   /SetVariables(Emitter.Particle Attribute Reader=DataInterface() {propertyValues {EmitterBinding {bindingMode=Other, emitterName=HangingLanterns}}}) {isSetParameters=true}
  ]
  particleSpawnScript [
   &PartitionParticles(Number Of Partitions=3)
   &self:GetParentInfo(
    Hue Adjustment:NiagaraFloat=&RandomRangeFloat(Minimum=-0.1, Maximum=0.1)
    Parent Attribute Reader:NiagaraDataInterfaceParticleRead=$Emitter.Particle Attribute Reader
    ParentID:NiagaraInt32=$Particles.PartitionParticles.Partition
    Value Adjustment=1
   ) {module=GetParentInfo}
   /SetVariables(Particles.SparkleOffset:Vector3f=&RandomVector(Vector Scale=12)) {isSetParameters=true}
   &InitializeParticle(
    Color=$Particles.InheritedColor
    Lifetime Max=1.5
    Lifetime Min=0.1
    Lifetime Mode=::NewEnumerator1
    Position=$Particles.ParentMidPosition
    Position Mode=::NewEnumerator1
    Position Offset=$Particles.SparkleOffset
    Sprite Size Mode=::NewEnumerator3
    UsePositionOffset=true
    Write Material Random=true
   )
   &CameraOffset(Camera Offset Amount=$System.OffsetLanterns)
   &InheritVelocity(
    Inherited Velocity Amount Scale=<0.5 0.5 0.5>
    Speed Limit=true
    UseSpeedThreshold=false
    Velocity Source=&self:GetParentVelocity()
   )
  ]
  particleUpdateScript [
   &ParticleState()
   &SpriteRotationRate(Rotation Rate=&RandomRangeFloat(Minimum=-300, Maximum=300))
   &ScaleSpriteSize(
    Scale Sprite Size Mode=::NewEnumerator0
    Uniform Scale Factor=&FloatFromCurve(FloatCurve=DataInterface() {propertyValues {Curve {keys [
         {arriveTangent=3.1235, interpMode=RCIM_Cubic, leaveTangent=3.1235, tangentMode=RCTM_User, time=0, value=0}
         {interpMode=RCIM_Cubic, tangentMode=RCTM_User, time=0.5, value=1}
         {arriveTangent=-3.4019, tangentMode=RCTM_User, time=1, value=0}
        ]}}})
   )
   &Drag(Drag=&RandomRangeFloat(Minimum=1, Maximum=2), Use Linear Drag=true, Use Rotational Drag=false)
   &GravityForce(Gravity=&RandomRangeVector(Minimum=<0 0 -50>, Maximum=<0 0 -25>))
   &VortexForce(Influence Falloff Radius=25, Use Influence Falloff=true, UseOriginOffset=false, Vortex Origin:NiagaraPosition=$Particles.ParentMidPosition)
   &PointAttractionForce(
    Attraction Strength=5
    Attractor Position:NiagaraPosition=$Particles.ParentMidPosition
    Kill Radius=1
    Kill Within Radius=false
    Module.Use Falloff=true
    UseAttractorOffset=false
   )
   &SolveForcesAndVelocity(Clamp Velocity=false, Limit Acceleration=false)
   &self:GetParentInfo(
    Hue Adjustment:NiagaraFloat=&RandomRangeFloat(Minimum=-0.2, Maximum=0.2)
    Normalize Color=true
    Parent Attribute Reader:NiagaraDataInterfaceParticleRead=$Emitter.Particle Attribute Reader
    ParentID:NiagaraInt32=$Particles.PartitionParticles.Partition
    Value Adjustment:NiagaraFloat=&FloatFromCurve(FloatCurve=DataInterface() {propertyValues {Curve {keys [{arriveTangent=-27.5457, interpMode=RCIM_Cubic, leaveTangent=-27.5457, tangentMode=RCTM_User, time=0, value=5}, <at 0.5 1 RCIM_Cubic>]}}})
   ) {module=GetParentInfo001}
   /SetVariables(Particles.GlowColor=&MultiplyLinearColor(Linear Color=$Particles.Color, Scale Factor=<1 1 1 0.5>)) {isSetParameters=true}
   /SetVariables(Particles.SparklePositionMod:NiagaraPosition=&LerpPosition(
     A=&AddVectorToPosition(Position=$Particles.ParentMidPosition, Vector=$Particles.SparkleOffset)
     B=$Particles.Position
     Alpha=&FloatFromCurve(FloatCurve=DataInterface() {propertyValues {Curve {keys [{arriveTangent=2.4562, interpMode=RCIM_Cubic, leaveTangent=2.4562, tangentMode=RCTM_User, time=0, value=0}, <at 1 1>]}}})
    )) {enabled=false, isSetParameters=true}
  ]
  rendererClasses [&NiagaraSpriteRendererProperties, &NiagaraMeshRendererProperties]
  renderers [
   {rendererClass=&NiagaraSpriteRendererProperties}
   {rendererClass=&NiagaraSpriteRendererProperties, rendererIndex=1}
   {rendererClass=&NiagaraMeshRendererProperties, rendererIndex=2}
  ]
 }
}
summary {
 systemName=NS_Aura_FloatingLantern
 emitters [
  {emitterName=BodyGlow, rendererClasses [&NiagaraSpriteRendererProperties]}
  {bEnabled=false, emitterName=FN_Debug, rendererClasses [&NiagaraSpriteRendererProperties]}
  {
   emitterName=HangingLanterns
   rendererClasses [
    &NiagaraMeshRendererProperties
    &NiagaraLightRendererProperties
    &NiagaraSpriteRendererProperties
   ]
  }
  {bEnabled=false, emitterName=Minimal, rendererClasses [&NiagaraSpriteRendererProperties]}
  {emitterName=Sparkles, rendererClasses [&NiagaraSpriteRendererProperties, &NiagaraMeshRendererProperties]}
  {emitterName=Petals, simTarget=GPUComputeSim, rendererClasses [&NiagaraSpriteRendererProperties]}
  {emitterName=Ribbons, rendererClasses [&NiagaraRibbonRendererProperties, &NiagaraSpriteRendererProperties]}
 ]
 userVariables [
  User.bIsFrontEnd {type=NiagaraBool}
  User.bIsFrontEndPreview {type=NiagaraBool}
  User.IsFrontEnd {type=NiagaraBool}
  User.IsLargeBody {type=NiagaraBool}
 ]
}
